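# Niagara system for ice cream drips falling from the three nozzles of the Clyde soft serve machine prop.
# Drips burst-spawns short-lived sprites partitioned three ways, each partition placed at its nozzle socket via a socket reader.
# A scratch input colors them purple or yellow by partition, with a random bool for the middle nozzle.
# Sprites grow non-uniformly under half gravity; the Temp_ref emitter is disabled.
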NiagaraSystem /Clyde/Athena/Effects/Enviornment/NS_ClydeDrippingIceCream
dependencies {
 usedDataInterfaces [
  &NiagaraDataInterfaceSocketReader
  &NiagaraDataInterfaceCurve
  &NiagaraDataInterfaceVector2DCurve
 ]
 usedDynamicInputs [
  &Lerp_Float
  &FloatFromCurve
  &RandomRangeFloat
 ]
 usedModules [&SystemState]
}
emitters {
 Drips {
  emitterUpdateScript [
   &EmitterState()
   &SpawnBurst_Instantaneous(Use Spawn Probability=false, UseLoopCountLimit=false)
   &SpawnRate(SpawnRate=300, Use Spawn Probability=false) {enabled=false}
  ]
  particleSpawnScript [
   &PartitionParticles(Number Of Partitions=3, Partition Mode:ENiagara_PartitionMode=::NewEnumerator5)
   &InitializeParticle(
    Color=&self:ScratchDynamicInput(
     Purple:LinearColor=<rgba 0.175 0.0437 0.1662 0.9>
     Yellow:LinearColor=<rgba 0.175 0.16 0.0035 0.9>
     Partition:NiagaraInt32=$Particles.PartitionParticles.Partition
     MiddleColorSelecrt:NiagaraBool=&RandomBool()
    )
    Lifetime=0.4
    Sprite Size Max=<1.75 10>
    Sprite Size Min=<2.5 15>
    Sprite Size Mode=::NewEnumerator2
    UsePositionOffset=false
    Write Material Random=true
   )
   &SocketLocation(
    SamplePosOffset=true
    Socket Index=$Particles.PartitionParticles.Partition
    Socket Reader=DataInterface() {propertyValues {
      AttachComponentClass=null
      AttachComponentTag=None
      EditorPreviewAsset=/Clyde/Athena/BuildingActors/Props/Building/ActorBlueprints/BerryTart/soft_serve_machine_01_Clyde
      SourceActorRef=null
      SourceAsset=&soft_serve_machine_01_Clyde
      SourceMode=SourceOnly
      bRequireCurrentFrameData=true
      bUpdateSocketsPerFrame=true
      FilteredSockets [Nozzle_L_Purple, Nozzle_M_Both, Nozzle_R_Yellow]
      ObjectParameterBinding {parameter {name=None, typeDefHandle {registeredTypeIndex=1}, varData []}}
     }}
    Socket Sampling Mode=::NewEnumerator3
    VelocitySpeedLimit=false
   )
  ]
  particleUpdateScript [
   &ParticleState()
   &ScaleSpriteSize(
    Non-Uniform Curve Sprite Scale=DataInterface() {propertyValues {
      XCurve {keys [{tangentMode=RCTM_User, time=0, value=0.75}, {tangentMode=RCTM_User, time=1, value=1}]}
      YCurve {keys [{tangentMode=RCTM_User, time=0, value=0.5}, {tangentMode=RCTM_User, time=1, value=1}]}
     }}
    Scale Sprite Size Mode=::NewEnumerator3
   )
   &GravityForce(Gravity=&Multiply_VectorByFloat(Vector=<0 0 -980>, Float=0.5))
   &SolveForcesAndVelocity(Clamp Velocity=false, Limit Acceleration=false)
  ]
  rendererClasses [&NiagaraSpriteRendererProperties]
  renderers [{rendererClass=&NiagaraSpriteRendererProperties}]
 }
 Temp_ref {
  bEnabled=false
  emitterUpdateScript [
   &EmitterState()
   &SpawnBurst_Instantaneous(Use Spawn Probability=false, UseLoopCountLimit=false)
  ]
  particleSpawnScript [&InitializeParticle(Color Mode=::NewEnumerator0, Lifetime=5, UsePositionOffset=false, Write Material Random=true)]
  particleUpdateScript [&ParticleState()]
  rendererClasses [&NiagaraSpriteRendererProperties, &NiagaraComponentRendererProperties]
  renderers [
   {rendererClass=&NiagaraSpriteRendererProperties}
   {rendererClass=&NiagaraComponentRendererProperties, rendererIndex=1}
  ]
 }
}
summary {
 systemName=NS_ClydeDrippingIceCream
 emitters [
  {emitterName=Drips, rendererClasses [&NiagaraSpriteRendererProperties]}
  {
   bEnabled=false
   emitterName=Temp_ref
   rendererClasses [&NiagaraSpriteRendererProperties, &NiagaraComponentRendererProperties]
  }
 ]
 userVariables []
}
