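# Niagara system for soda dripping from the four nozzles of the Clyde soda machine static mesh.
# One sprite emitter bursts particles, partitioned four ways so each index selects both a filtered nozzle socket and its drip color.
# Drops grow via a size curve while heavy early drag holds them at the nozzle, then fall under halved gravity.
# Particles die by lifetime or inside a kill volume.
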NiagaraSystem /Clyde/Athena/Effects/Enviornment/NS_ClydeDrippingSoda
dependencies {
 usedDataInterfaces [
  &NiagaraDataInterfaceArrayColor
  &NiagaraDataInterfaceSocketReader
  &NiagaraDataInterfaceCurve
  &NiagaraDataInterfaceVector2DCurve
 ]
 usedDynamicInputs [&Lerp_Float, &RandomRangeFloat]
 usedModules [&SystemState]
}
emitters {Drips {
  emitterUpdateScript [
   &EmitterState()
   &SpawnBurst_Instantaneous(Use Spawn Probability=false, UseLoopCountLimit=false)
   &SpawnRate(SpawnRate=2, Use Spawn Probability=false) {enabled=false}
  ]
  particleSpawnScript [
   &PartitionParticles(Number Of Partitions=4, Partition Mode:ENiagara_PartitionMode=::NewEnumerator5)
   &InitializeParticle(
    Color=&SelectLinearColorFromArray(
     Array Sampling Mode=::NewEnumerator1
     Direct Array Index=$Particles.PartitionParticles.Partition
     Color Selection Array=DataInterface() {propertyValues {ColorData [<rgba 0.011 0.7913 0.7157 0.9>, <rgba 0.5 0.457 0.01 0.9>, <rgba 0.04 0.0071 0.0028 1>, <rgba 0.7 1 1 0.05>]}}
     Random Seed=?
    )
    Lifetime=1.5
    Sprite Size Max=<1.75 10>
    Sprite Size Min=<2.5 15>
    Sprite Size Mode=::NewEnumerator2
    UsePositionOffset=false
    Write Material Random=true
   )
   &SocketLocation(
    SamplePosOffset=true
    Sampled Position Offset=<-2 0 0>
    Socket Index=$Particles.PartitionParticles.Partition
    Socket Reader=DataInterface() {propertyValues {
      AttachComponentClass=null
      AttachComponentTag=None
      EditorPreviewAsset=/Clyde/Athena/BuildingActors/Props/Building/ActorBlueprints/Clayplug/SM_Commerce_SodaMachine_A_Clyde
      SourceActorRef=null
      SourceAsset=&SM_Commerce_SodaMachine_A_Clyde
      SourceMode=SourceOnly
      bRequireCurrentFrameData=true
      bUpdateSocketsPerFrame=true
      FilteredSockets [Soda_Nozzle_L_L, Soda_Nozzle_L_R, Soda_Nozzle_R_L, Soda_Nozzle_R_R]
      ObjectParameterBinding {parameter {name=None, typeDefHandle {registeredTypeIndex=1}, varData []}}
     }}
    Socket Sampling Mode=::NewEnumerator3
    VelocitySpeedLimit=false
   )
  ]
  particleUpdateScript [
   &ParticleState()
   &ScaleSpriteSize(
    Non-Uniform Curve Sprite Scale=DataInterface() {propertyValues {
      XCurve {keys [<at 0 0 RCIM_Cubic>, <at 0.2 0.7626>, {tangentMode=RCTM_User, time=1, value=1}]}
      YCurve {keys [<at 0 0 RCIM_Cubic>, <at 0.2 0.5253>, {tangentMode=RCTM_User, time=1, value=1}]}
     }}
    Scale Sprite Size Mode=::NewEnumerator3
   )
   &GravityForce(Gravity=&Multiply_VectorByFloat(Vector=<0 0 -980>, Float=0.5))
   &Drag(
    Drag=&FloatFromCurve(
     FloatCurve=DataInterface() {propertyValues {Curve {keys [
         {arriveTangent=-2.875, interpMode=RCIM_Constant, leaveTangentWeight=1.3155, time=0, value=1}
         {
          arriveTangentWeight=1.1296
          interpMode=RCIM_Cubic
          leaveTangentWeight=1.1296
          tangentMode=RCTM_User
          tangentWeightMode=RCTWM_WeightedBoth
          time=0.328
          value=0
         }
        ]}}}
     Scale Curve=1000
    )
    Use Linear Drag=true
    Use Rotational Drag=true
   )
   &SolveForcesAndVelocity(Clamp Velocity=false, Limit Acceleration=false)
   &KillParticlesInVolume(Kill Shape=::NewEnumerator2, Origin Offset:Vector3f=<0 0 5>)
  ]
  rendererClasses [&NiagaraSpriteRendererProperties]
  renderers [{rendererClass=&NiagaraSpriteRendererProperties}]
 }}
summary {
 systemName=NS_ClydeDrippingSoda
 emitters [{emitterName=Drips, rendererClasses [&NiagaraSpriteRendererProperties]}]
 userVariables []
}
