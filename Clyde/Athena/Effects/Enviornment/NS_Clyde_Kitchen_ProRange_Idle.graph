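# Niagara system for the idle effect on the Clyde kitchen pro range, placing burner flames, glow and steam at sockets on the range mesh.
# Flames arranges flickering meshes in rings on two right stovetop burners, with fewer per ring on some platform quality levels.
# Empty adds a blue light at each burner.
# Steam uses two spawn groups for rising sprites from the back vents and open left pot; its oven-door placement is disabled.
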NiagaraSystem /Clyde/Athena/Effects/Enviornment/NS_Clyde_Kitchen_ProRange_Idle
dependencies {
 usedDataInterfaces [&NiagaraDataInterfaceSocketReader, &NiagaraDataInterfaceCurve, &NiagaraDataInterfacePlatformSet]
 usedDynamicInputs []
 usedModules [&SystemState]
}
emitters {
 Empty {
  emitterUpdateScript [
   &EmitterState()
   &SpawnBurst_Instantaneous(Spawn Count=2, Use Spawn Probability=false, UseLoopCountLimit=false)
  ]
  particleSpawnScript [
   &PartitionParticles()
   &InitializeParticle(
    Color Mode=::NewEnumerator0
    Lifetime=$Emitter.CurrentLoopDuration
    Sprite Size Mode=::NewEnumerator3
    Uniform Sprite Size=1
    UsePositionOffset=false
    Write Material Random=true
   )
   &SocketLocation(
    SamplePosOffset=true
    Sampled Position Offset=<0 0 2>
    Socket Index=&ModuloInt(A=$Particles.PartitionParticles.Partition, B=2)
    Socket Reader=DataInterface() {propertyValues {
      AttachComponentClass=null
      AttachComponentTag=None
      EditorPreviewAsset=/Clyde/Athena/BuildingActors/Props/Building/ActorBlueprints/Clayplug/SM_Kitchen_Pro_Range_C_Clyde
      SourceActorRef=null
      SourceAsset=&SM_Kitchen_Pro_Range_C_Clyde
      SourceMode=SourceOnly
      bRequireCurrentFrameData=true
      bUpdateSocketsPerFrame=true
      FilteredSockets [OvenTopFire_R_TR, OvenTopFire_R_TL]
      ObjectParameterBinding {parameter {name=None, typeDefHandle {registeredTypeIndex=1}, varData []}}
     }}
    Socket Sampling Mode=::NewEnumerator3
    VelocitySpeedLimit=false
   )
  ]
  particleUpdateScript [
   &ParticleState()
   &FN_Light_Attributes(
    Color=<rgba 0.075 0.3833 1 1>
    Create Color Attribute=true
    Create Exponent Attribute=true
    Create Radius Attribute=true
    Exponent=2
    Radius=20
    Scale Color=7
   )
  ]
  rendererClasses [&NiagaraSpriteRendererProperties, &NiagaraLightRendererProperties]
  renderers [
   {rendererClass=&NiagaraSpriteRendererProperties}
   {rendererClass=&NiagaraLightRendererProperties, rendererIndex=1}
  ]
 }
 Flames {
  emitterSpawnScript [
   /SetVariables(Emitter.Num in ring:NiagaraInt32=&MakeCustomIntFromBool(
     Bool=&IsPlatformSetActive(New Platform Set=DataInterface() {propertyValues {Platforms {qualityLevelMask=-31}}})
     True INT=6
     False INT=&MakeCustomIntFromBool(
      Bool=&IsPlatformSetActive(New Platform Set=DataInterface() {propertyValues {Platforms {qualityLevelMask=-30}}})
      True INT=7
      False INT=&MakeCustomIntFromBool(
       Bool=&IsPlatformSetActive(New Platform Set=DataInterface() {propertyValues {Platforms {qualityLevelMask=-28}}})
       True INT=8
       False INT=10
      )
     )
    )) {isSetParameters=true}
   /SetVariables(Emitter.Num Rings=2) {isSetParameters=true}
  ]
  emitterUpdateScript [
   &EmitterState()
   &SpawnBurst_Instantaneous(
    Spawn Count=&Multiply_Int(A=$Emitter.Num in ring, B=$Emitter.Num Rings)
    Use Spawn Probability=false
    UseLoopCountLimit=true
   )
  ]
  particleSpawnScript [
   &PartitionParticles(
    Particles Per Partition:NiagaraInt32=$Emitter.Num in ring
    Partition Count Mode:ENiagara_InfiniteLoopDuration=::NewEnumerator1
    Partition Mode:ENiagara_PartitionMode=::NewEnumerator1
   )
   &InitializeParticle(
    Color Channel Mode:ENiagaraChannelCorrelation=::NewEnumerator2
    Color Maximum=<rgba 1 1 1 0.5>
    Color Minimum=<rgba 0.75 0.75 0.75 1>
    Color Mode=::NewEnumerator2
    Lifetime=$Emitter.CurrentLoopDuration
    Mesh Scale Mode=::NewEnumerator3
    Mesh Uniform Scale=0.065
    Sprite Size Mode=::NewEnumerator3
    Uniform Sprite Size=2
    UsePositionOffset=false
    Write Material Random=true
   )
   &SocketLocation(
    SamplePosOffset=true
    Sampled Position Offset=<0 0 -2>
    Socket Index=&ModuloInt(A=$Particles.PartitionParticles.Partition, B=2)
    Socket Reader=DataInterface() {propertyValues {
      AttachComponentClass=null
      AttachComponentTag=None
      EditorPreviewAsset=&SM_Kitchen_Pro_Range_C_Clyde
      SourceActorRef=null
      SourceAsset=&SM_Kitchen_Pro_Range_C_Clyde
      SourceMode=SourceOnly
      bRequireCurrentFrameData=true
      bUpdateSocketsPerFrame=true
      FilteredSockets [OvenTopFire_R_TR, OvenTopFire_R_TL]
      ObjectParameterBinding {parameter {name=None, typeDefHandle {registeredTypeIndex=1}, varData []}}
     }}
    Socket Sampling Mode=::NewEnumerator3
    VelocitySpeedLimit=false
   )
   &ShapeLocation(
    MaskSpawnGroup=false
    Override Seed=false
    Radius Position:NiagaraFloat=&RandomRangeFloat(Minimum=0.8, Maximum=1)
    Ring / Disc Distribution Mode:ENiagaraSphereDistributionMode=::NewEnumerator1
    Ring Radius=6
    Shape Primitive=::NewEnumerator4
    U Position=&Add_Float(
     A=$Particles.PartitionParticles.Normalized Index In Partition
     B=&Multiply_Float(A=0.025, B=&RandomRangeFloat(Minimum=-1, Maximum=1))
    )
    UseOwnerScale=true
   )
   &InitialMeshOrientation(
    Mesh Axis To Orient=<0 0 1>
    Mesh Orientation Mode=::NewEnumerator2
    Orientation Vector=&Lerp_Vector(A=$Particles.ShapeLocation.ShapeVector, B=<0 0 1>, Alpha=0.2)
    Use Rotation Vector=false
   )
   &InitialMeshOrientation(
    Mesh Orientation Mode=::NewEnumerator7
    Rotation=&Multiply_VectorByFloat(Vector=&RandomRangeVector(Minimum=<-1 -1 -1>, Maximum=<1 1 1>), Float=0.01)
    Use Rotation Vector=true
   ) {module=InitialMeshOrientation001}
  ]
  particleUpdateScript [
   &ParticleState(Kill Particles When Lifetime Has Elapsed=false, Let Infinitely Lived Particles Die When Emitter Deactivates=true)
   &ScaleMeshSize(Scale Factor=&Multiply_VectorByFloat(
     Vector=&MakeVector(X=1, Y=1, Z=&Waveform("Waveform [1]"=::NewEnumerator12, "[1] Frequency"=10, "[1] Amplitude Min/Max"=<0.9 1>))
     Float=1
    ))
   &ScaleMeshSize(Scale Factor=&VectorFromFloat(Value=&Waveform(
      Global Input Position=&RandomRangeFloat(Minimum=0, Maximum=5000000)
      "Waveform [1]"=::NewEnumerator12
      "[1] Phase Scale"=0.05
      "[1] Frequency"=1
      "[1] Amplitude Min/Max"=<0.65 1.25>
     ))) {module=ScaleMeshSize001}
  ]
  rendererClasses [&NiagaraSpriteRendererProperties, &NiagaraMeshRendererProperties]
  renderers [
   {rendererClass=&NiagaraSpriteRendererProperties}
   {rendererClass=&NiagaraMeshRendererProperties, rendererIndex=1}
  ]
 }
 Steam {
  emitterSpawnScript [/SetVariables(Emitter.Oven offset:Vector3f=<0 0 25>, Emitter.Pot Offset:Vector3f=<0 0 5>) {isSetParameters=true}]
  emitterUpdateScript [
   &EmitterState()
   &SpawnRate(SpawnRate=&Multiply_Float(A=3, B=2), Use Spawn Probability=false) {enabled=false}
   &SpawnRate(SpawnRate=35, Use Spawn Probability=false) {module=SpawnRate002}
   &SpawnRate(SpawnGroup=1, SpawnRate=1.5, Use Spawn Probability=false) {module=SpawnRate001}
  ]
  particleSpawnScript [
   &InitializeParticle(
    Color=&MaskLinearColorBySpawnGroup(
     Masked Linear Color=&RandomRangeLinearColor(Minimum=<rgba 1 1 1 0.0075>, Maximum=<rgba 1 1 1 0.01>)
     Passthrough Linear Color=&RandomRangeLinearColor(Minimum=<rgba 0.75 0.75 0.75 0.05>, Maximum=<rgba 1 1 1 0.035>)
    )
    Lifetime Max=3.5
    Lifetime Min=3
    Lifetime Mode=::NewEnumerator1
    Sprite Rotation Mode=::NewEnumerator0
    Sprite Size Max=<200 200>
    Sprite Size Min=<100 100>
    Sprite Size Mode=::NewEnumerator2
    Sprite UV Mode=::NewEnumerator1
    UsePositionOffset=false
    Write Material Random=true
   )
   &ApplyInitialForces(Apply Rotational Force to Rotational Velocity=false) {enabled=false}
   &AddVelocity(
    Coordinate Space=::World
    Scale Added Velocity=&VectorFromFloat(Value=1)
    Velocity=&Multiply_VectorByFloat(
     Vector=&RandomRangeVector(Minimum=<50 0 -50>, Maximum=<35 0 -35>)
     Float=&MaskFloatBySpawnGroup(Masked Float=1, Passthrough Float=0)
    )
   )
   &PartitionParticles(Number Of Partitions=&MaskIntBySpawnGroup(Masked Int=1, Passthrough Int=1, Spawn Group Mask=0))
   &SocketLocation(
    SamplePosOffset=true
    Sampled Position Offset=$Emitter.Oven offset
    Socket Reader=DataInterface() {propertyValues {
      AttachComponentClass=null
      AttachComponentTag=None
      EditorPreviewAsset=&SM_Kitchen_Pro_Range_C_Clyde
      SourceActorRef=null
      SourceAsset=&SM_Kitchen_Pro_Range_C_Clyde
      SourceMode=SourceOnly
      bRequireCurrentFrameData=false
      bUpdateSocketsPerFrame=false
      FilteredSockets [OvenDoor_L_L]
      ObjectParameterBinding {parameter {name=None, typeDefHandle {registeredTypeIndex=1}, varData []}}
     }}
    Socket Sampling Mode=::NewEnumerator3
    VelocitySpeedLimit=false
   )
   &SocketLocation(
    SamplePosOffset=true
    Sampled Position Offset=$Emitter.Oven offset
    Socket Reader=DataInterface() {propertyValues {
      AttachComponentClass=null
      AttachComponentTag=None
      EditorPreviewAsset=&SM_Kitchen_Pro_Range_C_Clyde
      SourceActorRef=null
      SourceAsset=&SM_Kitchen_Pro_Range_C_Clyde
      SourceMode=SourceOnly
      bRequireCurrentFrameData=false
      bUpdateSocketsPerFrame=false
      FilteredSockets [Vent_Back_Left]
      ObjectParameterBinding {parameter {name=None, typeDefHandle {registeredTypeIndex=1}, varData []}}
     }}
    Socket Sampling Mode=::NewEnumerator3
    VelocitySpeedLimit=false
   ) {module=SocketLocation006}
   &SocketLocation(
    SamplePosOffset=true
    Sampled Position Offset=$Emitter.Oven offset
    Socket Reader=DataInterface() {propertyValues {
      AttachComponentClass=null
      AttachComponentTag=None
      EditorPreviewAsset=&SM_Kitchen_Pro_Range_C_Clyde
      SourceActorRef=null
      SourceAsset=&SM_Kitchen_Pro_Range_C_Clyde
      SourceMode=SourceOnly
      bRequireCurrentFrameData=false
      bUpdateSocketsPerFrame=false
      FilteredSockets [Vent_Back_Right]
      ObjectParameterBinding {parameter {name=None, typeDefHandle {registeredTypeIndex=1}, varData []}}
     }}
    Socket Sampling Mode=::NewEnumerator3
    VelocitySpeedLimit=false
   ) {module=SocketLocation005}
   &SocketLocation(
    SamplePosOffset=true
    Sampled Position Offset=$Emitter.Oven offset
    Socket Reader=DataInterface() {propertyValues {
      AttachComponentClass=null
      AttachComponentTag=None
      EditorPreviewAsset=&SM_Kitchen_Pro_Range_C_Clyde
      SourceActorRef=null
      SourceAsset=&SM_Kitchen_Pro_Range_C_Clyde
      SourceMode=SourceOnly
      bRequireCurrentFrameData=false
      bUpdateSocketsPerFrame=false
      FilteredSockets [OvenDoor_L_R]
      ObjectParameterBinding {parameter {name=None, typeDefHandle {registeredTypeIndex=1}, varData []}}
     }}
    Socket Sampling Mode=::NewEnumerator3
    VelocitySpeedLimit=false
   ) {enabled=false, module=SocketLocation004}
   &SocketLocation(
    SamplePosOffset=true
    Sampled Position Offset=$Emitter.Oven offset
    Socket Reader=DataInterface() {propertyValues {
      AttachComponentClass=null
      AttachComponentTag=None
      EditorPreviewAsset=&SM_Kitchen_Pro_Range_C_Clyde
      SourceActorRef=null
      SourceAsset=&SM_Kitchen_Pro_Range_C_Clyde
      SourceMode=SourceOnly
      bRequireCurrentFrameData=false
      bUpdateSocketsPerFrame=false
      FilteredSockets [OvenDoor_R_L]
      ObjectParameterBinding {parameter {name=None, typeDefHandle {registeredTypeIndex=1}, varData []}}
     }}
    Socket Sampling Mode=::NewEnumerator3
    VelocitySpeedLimit=false
   ) {enabled=false, module=SocketLocation003}
   &SocketLocation(
    SamplePosOffset=true
    Sampled Position Offset=$Emitter.Oven offset
    Socket Reader=DataInterface() {propertyValues {
      AttachComponentClass=null
      AttachComponentTag=None
      EditorPreviewAsset=&SM_Kitchen_Pro_Range_C_Clyde
      SourceActorRef=null
      SourceAsset=&SM_Kitchen_Pro_Range_C_Clyde
      SourceMode=SourceOnly
      bRequireCurrentFrameData=false
      bUpdateSocketsPerFrame=false
      FilteredSockets [OvenDoor_R_R]
      ObjectParameterBinding {parameter {name=None, typeDefHandle {registeredTypeIndex=1}, varData []}}
     }}
    Socket Sampling Mode=::NewEnumerator3
    VelocitySpeedLimit=false
   ) {enabled=false, module=SocketLocation002}
   &SocketLocation(
    SamplePosOffset=true
    Sampled Position Offset=$Emitter.Pot Offset
    Socket Reader=DataInterface() {propertyValues {
      AttachComponentClass=null
      AttachComponentTag=None
      EditorPreviewAsset=&SM_Kitchen_Pro_Range_C_Clyde
      SourceActorRef=null
      SourceAsset=&SM_Kitchen_Pro_Range_C_Clyde
      SourceMode=SourceOnly
      bRequireCurrentFrameData=false
      bUpdateSocketsPerFrame=false
      FilteredSockets [Pot_Left_Open]
      ObjectParameterBinding {parameter {name=None, typeDefHandle {registeredTypeIndex=1}, varData []}}
     }}
    Socket Sampling Mode=::NewEnumerator3
    VelocitySpeedLimit=false
   ) {module=SocketLocation001}
   /SetVariables(Particles.Position=&self:ScratchDynamicInput(
     LeftDoorPos:NiagaraPosition=&Lerp_Vector(
      A=$Output.SocketLocation.SampledPosition
      B=$Output.SocketLocation004.SampledPosition
      Alpha=&RandomRangeFloat(Minimum=0.25, Maximum=0.75)
     )
     RightDoorPos:NiagaraPosition=&Lerp_Vector(
      A=$Output.SocketLocation003.SampledPosition
      B=$Output.SocketLocation002.SampledPosition
      Alpha=&RandomRangeFloat(Minimum=0.25, Maximum=0.75)
     )
     Pot:NiagaraPosition=$Output.SocketLocation001.SampledPosition
     Is Doors:NiagaraBool=/Niagara/DynamicInputs/Helpers/MaskBoolBySpawnGroup(Masked Bool:NiagaraBool=true, Passthrough Bool:NiagaraBool=false, Spawn Group Mask:NiagaraInt32=0)
     Left Door:NiagaraBool=&SetBoolByIntComparison(A=$Particles.PartitionParticles.Partition, B=0, Comparison Type=::NewEnumerator0)
    )) {enabled=false, isSetParameters=true}
   /SetVariables(Particles.Position=&MaskVectorBySpawnGroup(
     Masked Vector=&Lerp_Vector(
      A=$Output.SocketLocation005.SampledPosition
      B=$Output.SocketLocation006.SampledPosition
      Alpha=&RandomRangeFloat(Minimum=0, Maximum=0.8)
     )
     Passthrough Vector=Linked() {linkedVariable {name=Output.SocketLocation001.SampledPosition, type=NiagaraPosition}}
    )) {isSetParameters=true}
   &Color() {enabled=false}
   &ApplyInitialForces() {module=ApplyInitialForces001}
  ]
  particleUpdateScript [
   &ParticleState()
   &SpriteRotationRate(Rotation Rate=&Multiply_Float(A=30, B=&RandomRangeFloat(Minimum=-1, Maximum=1)))
   &CameraOffset(Camera Offset Amount=50)
   &AccelerationForce(Acceleration=&RandomRangeVector(Minimum=<0 0 50>, Maximum=<0 0 80>))
   &Drag(Drag=2, Use Linear Drag=true, Use Rotational Drag=false)
   &SolveForcesAndVelocity(Clamp Velocity=true, Limit Acceleration=false, Speed Limit=100)
   &ScaleColor(
    Scale Alpha=&FloatFromCurve(FloatCurve=DataInterface() {propertyValues {Curve {keys [
         {interpMode=RCIM_Cubic, tangentMode=RCTM_User, time=0, value=0}
         <at 0.35 0.9847 RCIM_Cubic>
         {arriveTangent=-0.0405, interpMode=RCIM_Cubic, leaveTangent=-0.0405, tangentMode=RCTM_User, time=1, value=0}
        ]}}})
    ScaleA=true
    ScaleRGB=true
   )
   &ScaleSpriteSizeBySpeed(Max Scale Factor=<1.1 1.1>, Velocity Threshold=100)
   &ScaleSpriteSize(
    Uniform Curve Scale=&MaskFloatBySpawnGroup(Masked Float=0.35, Passthrough Float=0.3)
    Uniform Curve Sprite Scale=DataInterface() {propertyValues {Curve {keys [{tangentMode=RCTM_User, time=0, value=0.66}, {tangentMode=RCTM_User, time=1, value=1}]}}}
   )
   &SubUVAnimation()
  ]
  rendererClasses [&NiagaraSpriteRendererProperties]
  renderers [{rendererClass=&NiagaraSpriteRendererProperties}]
 }
}
summary {
 systemName=NS_Clyde_Kitchen_ProRange_Idle
 emitters [
  {emitterName=Steam, rendererClasses [&NiagaraSpriteRendererProperties]}
  {emitterName=Flames, rendererClasses [&NiagaraSpriteRendererProperties, &NiagaraMeshRendererProperties]}
  {emitterName=Empty, rendererClasses [&NiagaraSpriteRendererProperties, &NiagaraLightRendererProperties]}
 ]
 userVariables []
}
