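# Niagara ambient fauna effect that spawns a steady trickle of circling seagulls on the GPU, drawn with mesh and sprite renderers.
# Birds start on a raised torus sized by User.SpawnRadius with tangential velocity, then vortex and curl noise forces shape their flight.
# Negative-strength point attraction at the camera position pushes nearby birds away.
# Mesh scale fades over lifetime and by camera distance, while dynamic material parameters add per-bird variation.
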NiagaraSystem /Clyde/Athena/Effects/Enviornment/NS_Fauna_Seagulls_CS_Circling_Clyde
dependencies {
 usedDataInterfaces [&NiagaraDataInterfaceCamera, &NiagaraDataInterfaceCurve]
 usedDynamicInputs []
 usedModules [&SystemState]
}
emitters {BirdsInteractabler {
  simTarget=GPUComputeSim
  emitterUpdateScript [
   /SetVariables(Emitter.CameraPosition:Vector3f=&ReturnCameraVector(Camera Property=::NewEnumerator0, Coordinate Space=::Simulation)) {isSetParameters=true}
   &EmitterState()
   &SpawnBurst_Instantaneous(Spawn Count=250, Use Spawn Probability=false) {enabled=false}
   &SpawnRate(
    SpawnRate=&Multiply_Float(
     A=10
     B=&FloatFromCurve(
      FloatCurve=DataInterface() {propertyValues {Curve {keys [<at 0.2027 1>,<at 0.2492 0>,<at 0.9975 0>]}}}
      CurveIndex=$Emitter.NormalizedLoopAge
      Scale Curve=&Multiply_Float(
       A=&NormalizeDistanceRange(
        Start Position:NiagaraPosition=Linked() {linkedVariable {name=Emitter.CameraPosition, type=Vector3f}}
        End Position:NiagaraPosition=$Engine.Owner.Position
        Distance:NiagaraFloat=10000
        Invert Normalized Range:NiagaraBool=false
       )
       B=1
      )
     )
    )
    Use Spawn Probability=false
   ) {enabled=false}
   &SpawnRate(SpawnRate=2, Use Spawn Probability=false) {module=SpawnRate001}
  ]
  particleSpawnScript [
   /Niagara/Modules/Spawn/Initialization/InitializeParticle(
    Lifetime=5
    Mesh Scale=&Lerp_Vector(A=<1 1 1>, B=<1.25 1.25 1.25>, Alpha=&UniformAOrBFloat(A=0))
    Module.Write Color=true
    Module.Write Lifetime=true
    Module.Write Mass=true
    Module.Write Material Random=true
    Module.Write Position=false
    Module.Write Scale=true
    Module.Write SpriteRotation=false
    Module.Write SpriteSize=false
    Position=&SimulationPosition()
   ) {enabled=false}
   /Niagara/Modules/Spawn/Initialization/V2/InitializeParticle(
    Lifetime=10
    Mesh Scale=&Multiply_VectorByFloat(Vector=&RandomRangeVector(Minimum=<1.25 1.25 1.25>, Maximum=<1 1 1>), Float=1.9)
    Mesh Scale Mode=::NewEnumerator4
    Position Offset=<0 0 500>
    UsePositionOffset=true
    Write Material Random=true
   ) {module=InitializeParticle001}
   &ApplyInitialForces()
   /Niagara/Modules/Spawn/Location/SphereLocation(MaskSpawnGroup=false, Module.Override Randomness=false, Non Uniform Scale=<1 1 0.4>, Sphere Radius=2000, Surface Only=true) {enabled=false}
   &DynamicMaterialParameters(
    Index 0 Param 2=&UniformRangedFloat(Minimum=1.25, Maximum=1.5)
    Module.Param0WriteEnabled=true
    Module.Param1WriteEnabled=true
    Module.Param2WriteEnabled=true
    Module.Param3WriteEnabled=true
   )
   &TorusLocation(
    Align Torus With Axis=false
    Enable Torus Knot Radius=false
    Handle Radius=1000
    Large Radius:NiagaraFloat=$User.SpawnRadius
    MaskSpawnGroup=false
    Module.Override Randomness=false
    Torus Mode=::NewEnumerator1
   )
   &AddVelocity(Velocity=$Output.TorusLocation.TorusTangent, Velocity Speed Scale=500)
  ]
  particleUpdateScript [
   &ParticleState()
   &Color()
   /Niagara/Modules/Update/Forces/PointAttractionForce(Attraction Radius=256, Module.Kill Within Radius=false, Module.Use Falloff=true) {enabled=false}
   /Niagara/Modules/Update/Forces/V2/PointAttractionForce(
    Attraction Radius=2000
    Attraction Strength=-25
    Attractor Position:Vector3f=$Emitter.CameraPosition
    Kill Within Radius=false
    Module.Use Falloff=true
    UseAttractorOffset=false
   ) {module=PointAttractionForce001}
   &VortexForce(Origin Pull Amount=0.5, Vortex Force Amount=125)
   &CurlNoiseForce(Curl Noise Cone Mask Angle=90, Mask Curl Noise=true, Noise Frequency=15000, Noise Strength=250, Pan Noise=false)
   &SolveForcesAndVelocity(Clamp Velocity=true, Limit Acceleration=true)
   &ScaleMeshSize(Scale Factor=&Multiply_Vector(
     A=<1 1 1>
     B=&VectorFromFloat(Value=&Multiply_Float(
       A=&FloatFromCurve(FloatCurve=DataInterface() {propertyValues {Curve {keys [
            {arriveTangent=1.7794, interpMode=RCIM_Cubic, leaveTangent=1.7794, tangentMode=RCTM_User, time=-0.0063, value=0.4982}
            {arriveTangent=0.3677, interpMode=RCIM_Cubic, leaveTangent=0.3677, tangentMode=RCTM_User, time=0.2892, value=0.9107}
            {arriveTangent=-1.2812, interpMode=RCIM_Cubic, leaveTangent=-1.2812, tangentMode=RCTM_User, time=0.8446, value=0.8923}
            <at 0.9985 0.5165>
           ]}}})
       B=&FloatFromCurve(
        FloatCurve=DataInterface() {propertyValues {Curve {keys [<at 0.2476 0>, {interpMode=RCIM_Cubic, tangentMode=RCTM_User, time=0.3555, value=1.2901}, <at 0.8976 2.0314>, <at 1 0>]}}}
        CurveIndex=&NormalizeDistanceRange(
         Start Position:NiagaraPosition=Linked() {linkedVariable {name=Emitter.CameraPosition, type=Vector3f}}
         End Position:NiagaraPosition=Linked() {linkedVariable {name=Particles.FN_Orbit_v4.Position, type=Vector3f}}
         Distance:NiagaraFloat=15000
         Invert Normalized Range:NiagaraBool=false
        )
        Scale Curve=2
       )
      ))
    )) {enabled=false}
   &ScaleMeshSize(Scale Factor=&Multiply_Vector(
     A=<1 1 1>
     B=&VectorFromFloat(Value=&Multiply_Float(
       A=&FloatFromCurve(
        FloatCurve=DataInterface() {propertyValues {Curve {keys [
            {arriveTangent=1.7794, interpMode=RCIM_Cubic, leaveTangent=1.7794, tangentMode=RCTM_User, time=-0.0063, value=0}
            {arriveTangent=0.3677, interpMode=RCIM_Cubic, leaveTangent=0.3677, tangentMode=RCTM_User, time=0.1297, value=0.8851}
            {arriveTangent=-1.2812, interpMode=RCIM_Cubic, leaveTangent=-1.2812, tangentMode=RCTM_User, time=0.8446, value=0.8923}
            <at 0.9985 0>
           ]}}}
        Scale Curve=0.66
       )
       B=&FloatFromCurve(
        FloatCurve=DataInterface() {propertyValues {Curve {keys [<at -0.0055 -0.0388>, {interpMode=RCIM_Cubic, tangentMode=RCTM_User, time=0.0645, value=2}, <at 0.9 2>, <at 1 0>]}}}
        CurveIndex=&NormalizeDistanceRange(
         Start Position:NiagaraPosition=Linked() {linkedVariable {name=Emitter.CameraPosition, type=Vector3f}}
         End Position:NiagaraPosition=$Particles.Position
         Distance:NiagaraFloat=50000
         Invert Normalized Range:NiagaraBool=false
        )
       )
      ))
    )) {module=ScaleMeshSize001}
   /Niagara/Modules/Update/Camera/MaintainInCameraParticleScale() {enabled=false}
   &DynamicMaterialParameters(
    Index 0 Param 1=&RandomRangeFloat(Minimum=0, Maximum=10)
    Index 0 Param 2=2
    Module.Param0WriteEnabled=true
    Module.Param1WriteEnabled=true
    Module.Param2WriteEnabled=true
    Module.Param3WriteEnabled=true
   ) {module=DynamicMaterialParameters001}
  ]
  rendererClasses [&NiagaraMeshRendererProperties, &NiagaraSpriteRendererProperties]
  renderers [
   {rendererClass=&NiagaraMeshRendererProperties, rendererIndex=0}
   {rendererClass=&NiagaraSpriteRendererProperties, rendererIndex=1}
  ]
 }}
summary {
 systemName=NS_Fauna_Seagulls_CS_Circling_Clyde
 emitters [{
   emitterName=BirdsInteractabler
   simTarget=GPUComputeSim
   rendererClasses [&NiagaraMeshRendererProperties, &NiagaraSpriteRendererProperties]
  }]
 userVariables [User.SpawnRadius {type=NiagaraFloat, defaultValue=3000, description=""}]
}
